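; FortPlayerControllerPvE_BP is the Blueprint player controller for PvE that adds audio feedback when the fullscreen map toggles.
; It handles the MiniMapComponent's OnToggleFullscreenMap event and plays either the Map Open Sound or Map Close Sound variable as 2D audio, based on the visibility flag.
; The construction script is empty, and the file defines no other logic.
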(event OnToggleFullscreenMap(MiniMapComponent)
  (params (bFullscreenMapVisible :type Boolean))
  (if bFullscreenMapVisible
    (Audio|PlaySound2D (Variables|Default|GetMap%20Open%20Sound))
    (else
      (Audio|PlaySound2D (Variables|Default|GetMap%20Close%20Sound)))))

(fn ConstructionScript)
